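# Translucent unlit material that draws a moon disc with a surrounding glow, named for the Arid variant in Fortress_Sky.
# Exponential radial gradients on UV1 form a tight moon mask and a wide halo, whose powered copy adds a stronger inner glow.
# These layers are summed into a combined mask.
# The moon mask also blends the outer glow color into a tinted moon texture scaled about its center, with parameters for sizes, amounts and colors.
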Material /Game/Packages/Fortress_Sky/MaterialInstance/Arid/M_Moon_Arid
graph {
 properties {blendMode=BLEND_Translucent, pixelDepthOffsetMode=PDOM_Legacy, shadingModel=MSM_Unlit, usedShadingModels=Unlit}
 nodes [
  n1 = MaterialFunctionCall(UVs=TextureCoordinate() {coordinateIndex=1}, Radius=Constant() {r=0.5}, Density=Constant() {r=1.05}) {materialFunction=&RadialGradientExponential}
  n2 = ScalarParameter() {defaultValue=0.15, parameterName=OuterGlowAmount}
  n3 = Power(Base=@n1.RadialGradientExponential, Exp=Constant() {r=8})
  n4 = ScalarParameter() {defaultValue=20, parameterName=InnerGlowAmount}
  n5 = MaterialFunctionCall(
   UVs=TextureCoordinate() {coordinateIndex=1}
   Radius=ScalarParameter() {defaultValue=0.075, parameterName=MoonMaskRadius}
   Density=ScalarParameter() {defaultValue=75, parameterName=MoonMaskDensity}
  ) {materialFunction=&RadialGradientExponential}
  Add(
   A=Add(A=Multiply(A=@n1.RadialGradientExponential, B=@n2) {constA=0.1}, B=Multiply(A=@n3, B=@n4) {constA=0.1})
   B=@n5.RadialGradientExponential
  )
  n6 = MaterialFunctionCall(UVs=TextureCoordinate() {coordinateIndex=1}, Texture Scale=ScalarParameter() {defaultValue=0.15, parameterName=MoonDiffuseScale}) {materialFunction=&ScaleUVsByCenter}
  LinearInterpolate(
   A=VectorParameter() {parameterName=OuterGlowColor, defaultValue=<rgba 0.5748 0.5817 1 1>}.RGB
   B=Multiply(
    A=TextureSample(UVs=@n6.UVs) {texture=&/T_Moon}.RGB
    B=VectorParameter() {parameterName=DiffuseShift, defaultValue=<rgba 1.8642 1.9742 5 1>}.RGB
   )
   Alpha=@n5.RadialGradientExponential
  )
 ]
}
