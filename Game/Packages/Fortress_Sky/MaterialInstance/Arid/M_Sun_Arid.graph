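# Translucent unlit sun material for the Arid sky, previewed on a flat sun/moon mesh.
# Exponential radial gradients on UV1 form the sun disc and inner and outer glows.
# Flares come from vertex-color red height-lerped against gradients, then broken up by two counter-panning cloud-noise samples masked by inverted vertex-color green.
# The combined mask blends two Color_Time_of_Day palettes with HDR morning, day, evening and night colors; a similar sum forms an opacity-like mask.
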Material /Game/Packages/Fortress_Sky/MaterialInstance/Arid/M_Sun_Arid
graph {
 properties {
  blendMode=BLEND_Translucent
  pixelDepthOffsetMode=PDOM_Legacy
  previewMesh=/Game/Packages/Fortress_Sky/Mesh_NoLOD/SM_FlatSunMoon
  shadingModel=MSM_Unlit
  usedShadingModels=Unlit
 }
 nodes [
  n1 = MaterialFunctionCall(
   UVs=TextureCoordinate() {coordinateIndex=1}
   Radius=ScalarParameter() {defaultValue=0.5, parameterName=GlowAmt_10}
   Density=ScalarParameter() {defaultValue=1, parameterName=GlowAmt_11}
  ) {materialFunction=&RadialGradientExponential}
  n2 = Add(
   A=@n1.RadialGradientExponential
   B=MaterialFunctionCall(
    UVs=TextureCoordinate() {coordinateIndex=1}
    Radius=ScalarParameter() {defaultValue=0.07, parameterName=GlowAmt_15}
    Density=ScalarParameter() {defaultValue=2, parameterName=GlowAmt_16}
   ) {materialFunction=&RadialGradientExponential}.RadialGradientExponential
  )
  n3 = VertexColor()
  n4 = Constant() {r=20}
  n5 = Clamp(None=Multiply(A=Power(Base=@n3.R, Exp=@n4), B=Constant() {r=0.5}))
  n6 = Power(Base=VertexColor().R, Exp=Constant() {r=50})
  n7 = Constant() {r=0.2}
  n8 = Multiply(
   A=ScalarParameter() {defaultValue=0.5, parameterName=FlareAmount}
   B=Add(
    A=MaterialFunctionCall(Transition Phase=@n2, Height Texture=@n5) {materialFunction=&HeightLerp}.Lerp Alpha No Contrast
    B=Multiply(A=@n6, B=@n7)
   )
  )
  n9 = OneMinus(None=VertexColor().G)
  n10 = Multiply(A=TextureCoordinate() {coordinateIndex=1}, B=Constant() {r=1.5})
  n11 = Time()
  n12 = Constant() {r=0.06}
  n13 = Add(A=@n10, B=Constant2Vector() {g=0.663, r=0.44})
  n14 = Multiply(A=@n12) {constB=-1}
  n15 = Multiply(
   A=TextureSample(UVs=Panner(Coordinate=@n10, Time=@n11, Speed=@n12)) {samplerType=SAMPLERTYPE_LinearGrayscale, texture=/Game/Global/Textures/Default/Tileable_Patterns/T_StormCloudDetail01_M}.R
   B=TextureSample(UVs=Panner(Coordinate=@n13, Time=@n11, Speed=@n14)) {samplerType=SAMPLERTYPE_LinearGrayscale, texture=&T_StormCloudDetail01_M}.R
  )
  n16 = Multiply(
   A=@n9
   B=MaterialFunctionCall(Transition Phase=@n9, Height Texture=LinearInterpolate(Alpha=@n15) {constA=-1}) {materialFunction=&HeightLerp}.Lerp Alpha No Contrast
  )
  n17 = ScalarParameter() {defaultValue=0.5, parameterName=OuterGlowAmount}
  n18 = Constant() {r=8}
  n19 = Add(
   A=Add(A=Multiply(A=@n8, B=@n16), B=Multiply(A=@n1.RadialGradientExponential, B=@n17) {constA=0.1})
   B=Multiply(A=Power(Base=@n1.RadialGradientExponential, Exp=@n18), B=ScalarParameter() {defaultValue=50, parameterName=InnerGlowAmount}) {constA=0.1}
  )
  n20 = MaterialFunctionCall(
   UVs=TextureCoordinate() {coordinateIndex=1}
   Radius=ScalarParameter() {defaultValue=0.05, parameterName=SunRadius}
   Density=ScalarParameter() {defaultValue=4, parameterName=SunDensity}
  ) {materialFunction=&RadialGradientExponential}
  Add(A=@n20.RadialGradientExponential, B=Multiply(A=@n19, B=Constant() {r=0.15}) {constA=0.1})
  n21 = Constant() {r=0.5}
  LinearInterpolate(
   A=MaterialFunctionCall(
    Color_Morning=VectorParameter() {parameterName=Morning Color_1, defaultValue=<rgba 1 0.3936 0.4434 1>}.RGB
    Color_Day=VectorParameter() {parameterName=Day Color_1, defaultValue=<rgba 1 0.2828 0.15 1>}.RGB
    Color_Evening=VectorParameter() {parameterName=Evening Color_1, defaultValue=<rgba 5 1.6552 0.2399 1>}.RGB
    Color_Night=VectorParameter() {parameterName=Night Color_1, defaultValue=<rgba 0.0296 0.1275 0.46 1>}.RGB
   ) {materialFunction=&Color_Time_of_Day}.Result
   B=MaterialFunctionCall(
    Color_Morning=VectorParameter() {parameterName=Morning Color_2, defaultValue=<rgba 50 13.5126 8.5258 1>}.RGB
    Color_Day=VectorParameter() {parameterName=Day Color_2, defaultValue=<rgba 4 3.7165 1.44 1>}.RGB
    Color_Evening=VectorParameter() {parameterName=Evening Color_2, defaultValue=<rgba 15 2.0877 0 1>}.RGB
    Color_Night=VectorParameter() {parameterName=Night Color_2, defaultValue=<rgba 0.6403 0.9255 1.2 1>}.RGB
   ) {materialFunction=&Color_Time_of_Day}.Result
   Alpha=Add(A=@n20.RadialGradientExponential, B=Multiply(A=@n21, B=@n19) {constA=0.1})
  )
 ]
}
